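# Sound cue for the premium supply drop entering a rift, built from drop pod rift-entry waves.
# A mixer layers two random picks: one of two B variants at natural pitch, and one of two C variants through a modulator that varies pitch.
# It uses tidal crane drop pod high-frequency alert attenuation and the Athena ambient sound class, with virtualization disabled and a concurrency override capping simultaneous instances.
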SoundCue /PremiumSupplyDrop/Assets/Audio/SC_PremiumSupplyDrop_EnterRift
root=Mixer_1
nodes [
 {id=Mixer_1, children [Modulator_2,Random_4], properties {inputVolume [1,1]}}
 {id=Modulator_2, children [Random_5], properties {pitchMax=1.3, pitchMin=0.9, volumeMax=1, volumeMin=1}}
 {id=Random_4, children [WavePlayer_9, WavePlayer_10], properties {weights [1,1]}}
 {id=Random_5, children [WavePlayer_11, WavePlayer_12], properties {bShouldExcludeFromBranchCulling=true, weights [1,1]}}
 {id=WavePlayer_10, properties {soundWaveAssetPtr=/DropPodParentGameplay/Assets/Audio/wav/TidalCraneDropPod_EnterRiftB_02}}
 {id=WavePlayer_11, properties {soundWaveAssetPtr=/DropPodParentGameplay/Assets/Audio/wav/TidalCraneDropPod_EnterRiftC_01}}
 {id=WavePlayer_12, properties {soundWaveAssetPtr=/DropPodParentGameplay/Assets/Audio/wav/TidalCraneDropPod_EnterRiftC_02}}
 {id=WavePlayer_9, properties {soundWaveAssetPtr=/DropPodParentGameplay/Assets/Audio/wav/TidalCraneDropPod_EnterRiftB_01}}
]
properties {
 attenuationSettings=&/ATT_TidalCraneDropPod_Alert_HighFreq001
 maxDistance=230000
 soundClassObject=&Fort_Ambient_Athena
 virtualizationMode=Disabled
 volumeMultiplier=1
 concurrencyOverrides {maxCount=4}
}
