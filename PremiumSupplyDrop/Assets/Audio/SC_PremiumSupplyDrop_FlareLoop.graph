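# SoundCue for the looping flare sound of the Premium Supply Drop.
# A single Wave Player feeds the output directly, looping the TidalCraneDropPod flare wave with no modulation or mixing nodes.
# The cue plays at reduced volume in the Athena ambient sound class and uses a close-range flare attenuation asset for distance falloff.
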SoundCue /PremiumSupplyDrop/Assets/Audio/SC_PremiumSupplyDrop_FlareLoop
root=WavePlayer_0
nodes [{id=WavePlayer_0, properties {bLooping=true, soundWaveAssetPtr=/PremiumSupplyDrop/Assets/Audio/wav/TidalCraneDropPod_Flare_Loop}}]
properties {
 attenuationSettings=&/ATT_SC_PremiumSupplyDrop_FlareClose
 maxDistance=1500
 soundClassObject=&Fort_Ambient_Athena
 volumeMultiplier=0.6
}
