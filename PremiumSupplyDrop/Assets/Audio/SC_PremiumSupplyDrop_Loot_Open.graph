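# Sound cue for opening the premium supply drop loot container.
# A mixer layers three random picks, each choosing evenly between two waves: gas release, a latch or chest open (slightly boosted), and a loot-open latch.
# It plays on the loot reward sound class and references the shared item-interact attenuation, with override values that include occlusion tracing.
# Its concurrency override sets the resolution rule to prevent new instances.
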SoundCue /PremiumSupplyDrop/Assets/Audio/SC_PremiumSupplyDrop_Loot_Open
root=Mixer_0
nodes [
 {id=Mixer_0, children [Random_27,Random_2,Random_0], properties {inputVolume [1,1.2,1]}}
 {id=Random_0, children [WavePlayer_0,WavePlayer_5], properties {weights [1,1]}}
 {id=Random_2, children [WavePlayer_1,WavePlayer_4], properties {weights [1,1]}}
 {id=Random_27, children [WavePlayer_2,WavePlayer_3], properties {weights [1,1]}}
 {id=WavePlayer_0, properties {soundWaveAssetPtr=/PremiumSupplyDrop/Assets/Audio/wav/PremiumSupplyDrop_Loot_Open_Latch_01}}
 {id=WavePlayer_1, properties {soundWaveAssetPtr=/PremiumSupplyDrop/Assets/Audio/wav/PremiumSupplyDrop_LatchOpenB}}
 {id=WavePlayer_2, properties {soundWaveAssetPtr=/PremiumSupplyDrop/Assets/Audio/wav/PremiumSupplyDrop_GasRelease_01}}
 {id=WavePlayer_3, properties {soundWaveAssetPtr=/PremiumSupplyDrop/Assets/Audio/wav/PremiumSupplyDrop_GasRelease_02}}
 {id=WavePlayer_4, properties {soundWaveAssetPtr=/PremiumSupplyDrop/Assets/Audio/wav/PremiumSupplyDrop_ChestOpen}}
 {id=WavePlayer_5, properties {soundWaveAssetPtr=/PremiumSupplyDrop/Assets/Audio/wav/PremiumSupplyDrop_Loot_Open_Latch_02}}
]
properties {
 attenuationSettings=/Game/Sounds/Attenuation/Fort_Item_Interact_Att
 maxDistance=1850
 soundClassObject=&Fort_Loot_Reward_PSM
 volumeMultiplier=0.8
 attenuationOverrides {bEnableOcclusion=true, falloffDistance=1000, occlusionTraceChannel=ECC_GameTraceChannel11, attenuationShapeExtents {x=500}}
 concurrencyOverrides {resolutionRule=PreventNew}
}
