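# Looping sound cue for the premium supply drop targeting reticle, using close-range marker attenuation and the gadgets sound class.
# A mixer layers a continuous target loop with two quieter, repeating random bling layers, each with pitch and volume modulation.
# One bling layer repeats without a delay node, while the other waits a random few seconds between hits.
# The quality node feeds this mix to two levels and a single optimized loop to the other two.
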SoundCue /PremiumSupplyDrop/Assets/Audio/SC_PremiumSupplyDrop_ReticleLoop
root=QualityLevel_0
nodes [
 {id=Delay_0, children [Random_2], properties {delayMax=3, delayMin=2}}
 {id=Looping_0, children [Random_1]}
 {id=Looping_1, children [Delay_0]}
 {id=Mixer_0, children [WavePlayer_18, Modulator_1, Modulator_0], properties {inputVolume [1,0.35,0.35]}}
 {id=Modulator_0, children [Looping_1], properties {pitchMax=1.2, pitchMin=0.8, volumeMin=0.7}}
 {id=Modulator_1, children [Looping_0], properties {pitchMax=1.2, pitchMin=0.8, volumeMin=0.7}}
 {id=QualityLevel_0, children [Mixer_0, Mixer_0, WavePlayer_19, WavePlayer_19]}
 {id=Random_1, children [WavePlayer_14, WavePlayer_15, WavePlayer_16, WavePlayer_17], properties {weights [1,1,1,1]}}
 {id=Random_2, children [WavePlayer_10, WavePlayer_11, WavePlayer_12, WavePlayer_13], properties {weights [1,1,1,1]}}
 {id=WavePlayer_10, properties {soundWaveAssetPtr=/PremiumSupplyDrop/Assets/Audio/wav/Bling_Bus_OS_08}}
 {id=WavePlayer_11, properties {soundWaveAssetPtr=/PremiumSupplyDrop/Assets/Audio/wav/Bling_Bus_OS_07}}
 {id=WavePlayer_12, properties {soundWaveAssetPtr=/PremiumSupplyDrop/Assets/Audio/wav/Bling_Bus_OS_06}}
 {id=WavePlayer_13, properties {soundWaveAssetPtr=/PremiumSupplyDrop/Assets/Audio/wav/Bling_Bus_OS_05}}
 {id=WavePlayer_14, properties {soundWaveAssetPtr=/PremiumSupplyDrop/Assets/Audio/wav/Bling_Bus_OS_04}}
 {id=WavePlayer_15, properties {soundWaveAssetPtr=/PremiumSupplyDrop/Assets/Audio/wav/Bling_Bus_OS_03}}
 {id=WavePlayer_16, properties {soundWaveAssetPtr=/PremiumSupplyDrop/Assets/Audio/wav/Bling_Bus_OS_02}}
 {id=WavePlayer_17, properties {soundWaveAssetPtr=/PremiumSupplyDrop/Assets/Audio/wav/Bling_Bus_OS_01}}
 {id=WavePlayer_18, properties {bLooping=true, soundWaveAssetPtr=/PremiumSupplyDrop/Assets/Audio/wav/sfx_supplydrop_target_loop}}
 {id=WavePlayer_19, properties {bLooping=true, soundWaveAssetPtr=/PremiumSupplyDrop/Assets/Audio/wav/sfx_supplydrop_target_loop_opt}}
]
properties {
 attenuationSettings=&/Fort_SupplyDrop_Marker_Close_Athena_Att
 maxDistance=1500
 soundClassObject=&Fort_Gadgets
 volumeMultiplier=1
}
