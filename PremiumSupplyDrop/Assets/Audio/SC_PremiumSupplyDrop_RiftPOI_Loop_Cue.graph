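# Looping ambience cue for the Premium Supply Drop rift POI, with map-wide attenuation and a single-instance concurrency limit.
# A constant rift island loop mixes with two randomly delayed repeating layers: occasional island vox and more frequent distant thunder.
# A continuous modulator driven by RiftGrowth varies volume and pitch of those two layers, leaving the base loop unmodulated.
# A quality-level switch selects a larger or smaller thunder pool, with its remaining slots empty.
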SoundCue /PremiumSupplyDrop/Assets/Audio/SC_PremiumSupplyDrop_RiftPOI_Loop_Cue
root=Mixer_4
nodes [
 {id=Delay_1, children [Mixer_8], properties {delayMax=10, delayMin=0.5}}
 {id=Delay_2, children [Random_2], properties {delayMax=30, delayMin=10}}
 {id=Looping_1, children [Delay_1]}
 {id=Looping_2, children [Delay_2]}
 {id=Mixer_4, children [WavePlayer_1, ModulatorContinuous_1], properties {inputVolume [1,1]}}
 {id=Mixer_6, children [Looping_2,Looping_1], properties {inputVolume [1.8,1]}}
 {id=Mixer_8, children [QualityLevel_11], properties {inputVolume [0.4]}}
 {
  id=ModulatorContinuous_1
  children [Mixer_6]
  properties {pitchModulationParams {maxOutput=1.2, minOutput=0.8, parameterName=RiftGrowth}, volumeModulationParams {parameterName=RiftGrowth}}
 }
 {id=QualityLevel_11, children [Random_0,Random_1,null,null]}
 {
  id=Random_0
  children [WavePlayer_27, WavePlayer_31, WavePlayer_32, WavePlayer_33, WavePlayer_34, WavePlayer_35, WavePlayer_36, WavePlayer_37, WavePlayer_38]
  properties {weights [1,1,1,1,1,1,1,1,1]}
 }
 {id=Random_1, children [WavePlayer_39, WavePlayer_40, WavePlayer_41], properties {weights [1,1,1]}}
 {id=Random_2, children [WavePlayer_2,WavePlayer_3], properties {weights [1,1]}}
 {id=WavePlayer_1, properties {bLooping=true, soundWaveAssetPtr=/PremiumSupplyDrop/Assets/Audio/wav/s24_MMobj_Rift_Island_Loop}}
 {id=WavePlayer_2, properties {soundWaveAssetPtr=/PremiumSupplyDrop/Assets/Audio/wav/s24_MMobj_Island_Vox_01}}
 {id=WavePlayer_27, properties {soundWaveAssetPtr=/Game/Sounds/Fort_GamePlay_Sounds/Storms/Dist_Thunder_001}}
 {id=WavePlayer_3, properties {soundWaveAssetPtr=/PremiumSupplyDrop/Assets/Audio/wav/s24_MMobj_Island_Vox_02}}
 {id=WavePlayer_31, properties {soundWaveAssetPtr=/Game/Sounds/Fort_GamePlay_Sounds/Storms/Dist_Thunder_002}}
 {id=WavePlayer_32, properties {soundWaveAssetPtr=/Game/Sounds/Fort_GamePlay_Sounds/Storms/Dist_Thunder_003}}
 {id=WavePlayer_33, properties {soundWaveAssetPtr=/Game/Sounds/Fort_GamePlay_Sounds/Storms/Dist_Thunder_004}}
 {id=WavePlayer_34, properties {soundWaveAssetPtr=/Game/Sounds/Fort_GamePlay_Sounds/Storms/Dist_Thunder_005}}
 {id=WavePlayer_35, properties {soundWaveAssetPtr=/Game/Sounds/Fort_GamePlay_Sounds/Storms/Dist_Thunder_006}}
 {id=WavePlayer_36, properties {soundWaveAssetPtr=/Game/Sounds/Fort_GamePlay_Sounds/Storms/Dist_Thunder_007}}
 {id=WavePlayer_37, properties {soundWaveAssetPtr=/Game/Sounds/Fort_GamePlay_Sounds/Storms/Dist_Thunder_008}}
 {id=WavePlayer_38, properties {soundWaveAssetPtr=/Game/Sounds/Fort_GamePlay_Sounds/Storms/Dist_Thunder_009}}
 {id=WavePlayer_39, properties {soundWaveAssetPtr=&Dist_Thunder_001}}
 {id=WavePlayer_40, properties {soundWaveAssetPtr=&Dist_Thunder_002}}
 {id=WavePlayer_41, properties {soundWaveAssetPtr=&Dist_Thunder_003}}
]
properties {
 attenuationSettings=&/MMObj_Event_Mapwide_Att
 bOverrideConcurrency=true
 maxDistance=533000
 soundClassObject=&Fort_Gadgets
 volumeMultiplier=0.3
 concurrencyOverrides {maxCount=1}
}
