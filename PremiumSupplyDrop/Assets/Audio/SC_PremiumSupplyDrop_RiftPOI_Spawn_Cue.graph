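# Sound cue for a Rift POI spawn in the Premium Supply Drop content, layering close and distant audio.
# A mixer combines a single spawn wave under close attenuation with a louder distant layer under its own attenuation.
# The distant layer starts after a fixed one-second delay and randomly picks one of four rift-opening bursts, excluded from branch culling.
# The cue uses the gadgets sound class and overrides concurrency to cap simultaneous instances.
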SoundCue /PremiumSupplyDrop/Assets/Audio/SC_PremiumSupplyDrop_RiftPOI_Spawn_Cue
root=Mixer_2
nodes [
 {id=Attenuation_0, children [Random_1], properties {attenuationSettings=&/MMObj_POI_Distant_Att}}
 {id=Attenuation_1, children [WavePlayer_0], properties {attenuationSettings=&/MMObj_POI_Close_Att}}
 {id=Delay_0, children [Attenuation_0], properties {delayMax=1, delayMin=1}}
 {id=Mixer_2, children [Attenuation_1, Delay_0], properties {inputVolume [1,1.25]}}
 {
  id=Random_1
  children [WavePlayer_6, WavePlayer_7, WavePlayer_1, WavePlayer_10]
  properties {bShouldExcludeFromBranchCulling=true, weights [1,1,1,1]}
 }
 {id=WavePlayer_0, properties {soundWaveAssetPtr=/PremiumSupplyDrop/Assets/Audio/wav/MMObj_RiftPOI_Spawn_02}}
 {id=WavePlayer_1, properties {soundWaveAssetPtr=/PremiumSupplyDrop/Assets/Audio/wav/RiftEncounter_Open_Burst_Distant_03}}
 {id=WavePlayer_10, properties {soundWaveAssetPtr=/PremiumSupplyDrop/Assets/Audio/wav/RiftEncounter_Open_Burst_Distant_04}}
 {id=WavePlayer_6, properties {soundWaveAssetPtr=/PremiumSupplyDrop/Assets/Audio/wav/RiftEncounter_Open_Burst_Distant_01}}
 {id=WavePlayer_7, properties {soundWaveAssetPtr=/PremiumSupplyDrop/Assets/Audio/wav/RiftEncounter_Open_Burst_Distant_02}}
]
properties {
 bOverrideConcurrency=true
 maxDistance=81000
 soundClassObject=&Fort_Gadgets
 volumeMultiplier=0.7
 concurrencyOverrides {maxCount=6}
}
