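# Sound cue for the premium supply drop's searching sound, classed as loot collection audio with item-interact attenuation.
# A mixer layers a random opening click, a random servo, a pitch-modulated scan-success sound and a faint landing mechanism sound, with servos weighted loudest.
# Concurrency prevents new instances from starting once the limit is reached.
# Despite the SearchingLoop name, the graph contains no looping node.
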SoundCue /PremiumSupplyDrop/Assets/Audio/SC_PremiumSupplyDrop_SearchingLoop
root=Mixer_0
nodes [
 {id=Mixer_0, children [Random_0,Random_4,Modulator_0,WavePlayer_4], properties {inputVolume [0.3,0.5,0.2,0.05]}}
 {id=Modulator_0, children [WavePlayer_15], properties {pitchMin=1}}
 {id=Random_0, children [WavePlayer_0,WavePlayer_1,WavePlayer_2,WavePlayer_3], properties {weights [1,1,1,1]}}
 {id=Random_4, children [WavePlayer_23, WavePlayer_24, WavePlayer_25], properties {weights [1,1,1]}}
 {id=WavePlayer_0, properties {soundWaveAssetPtr=/PremiumSupplyDrop/Assets/Audio/wav/PremiumSupplyDrop_OpeningClick_01}}
 {id=WavePlayer_1, properties {soundWaveAssetPtr=/PremiumSupplyDrop/Assets/Audio/wav/PremiumSupplyDrop_OpeningClick_02}}
 {id=WavePlayer_15, properties {soundWaveAssetPtr=/PremiumSupplyDrop/Assets/Audio/wav/PremiumSupplyDrop_Scan_Success}}
 {id=WavePlayer_2, properties {soundWaveAssetPtr=/PremiumSupplyDrop/Assets/Audio/wav/PremiumSupplyDrop_OpeningClick_03}}
 {id=WavePlayer_23, properties {soundWaveAssetPtr=/PremiumSupplyDrop/Assets/Audio/wav/PremiumSupplyDrop_Servo_01}}
 {id=WavePlayer_24, properties {soundWaveAssetPtr=/PremiumSupplyDrop/Assets/Audio/wav/PremiumSupplyDrop_Servo_02}}
 {id=WavePlayer_25, properties {soundWaveAssetPtr=/PremiumSupplyDrop/Assets/Audio/wav/PremiumSupplyDrop_Servo_03}}
 {id=WavePlayer_3, properties {soundWaveAssetPtr=/PremiumSupplyDrop/Assets/Audio/wav/PremiumSupplyDrop_OpeningClick_04}}
 {id=WavePlayer_4, properties {soundWaveAssetPtr=/PremiumSupplyDrop/Assets/Audio/wav/PremiumSupplyDrop_LandingMech}}
]
properties {
 attenuationSettings=/Game/Sounds/Attenuation/Fort_Item_Interact_Mono_Att
 maxDistance=1500
 soundClassObject=&Fort_Loot_Collect
 volumeMultiplier=0.35
 concurrencyOverrides {resolutionRule=PreventNew}
}
